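; BP_BananaBus is a step-on rake trap that snaps up to whack the player who triggers it.
; An authority-checked overlap calls SnapRakeUp, which launches pawns found by a box check at the other end and runs a rotation timeline that fires the whack.
; Switchable modes make the whack stun, with a dot-product-chosen directional montage and timed input lock, knock back, or neither.
; Launches use granted gameplay abilities and send quest events, skipping players in vehicles.
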(event OnComponentBeginOverlap(RakeOverlap)
  (params (OverlappedComponent :type Primitive%20Component%20Object%20Reference) (OtherActor :type Actor%20Object%20Reference) (OtherComp :type Primitive%20Component%20Object%20Reference) (OtherBodyIndex :type Integer) (bFromSweep :type Boolean) (SweepResult :type Hit%20Result%20Structure%20%28by%20ref%29))
  ;[0 1 0.03773 0.5] RakeOverlap
  (Networking|SwitchHasAuthority
    (:Authority
      (bind _asfort_player_pawn (Utilities|Casting|CastToFortPlayerPawn OtherActor))
      (Variables|BananaBus|Whackback|SetWhackedPawn _asfort_player_pawn)
      (CallFunction|SnapRakeUp))))

(event Custom|SnapRakeUp
  (Networking|SwitchHasAuthority
    (:Authority
      ;[1 1 1] Disable Overlap and Toggle Rake mesh visibilities for animation
      (CallFunction|LaunchPlayersonStick)
      ;[1 1 1] Play Rake Animation
      (bind (_direction _rakealpha) (Timeline|RakeSnapAnimation PlayFromStart)
        (:Update
          (Transformation|SetRelativeRotation (Variables|Default|GetSM_BananaBusAnimated) (Math|Rotator|MakeRotator (Math|Float|Lerp 0.0 9e+01 _rakealpha)) false false))
        (:Finished
          ;[1 1 1] Play Rake Drop Audio
          (CallFunction|PlayDropAudio)
          ;[1 1 1] Reenable Overlap and Toggle Rake visibility for after animation
          (Variables|/Script/Engine.PrimitiveComponent|SetbGenerateOverlapEvents true (Variables|Default|GetRakeOverlap))
          (Rendering|SetVisibility (Variables|Debug|GetStaticMeshComponent) true false)
          (Rendering|SetVisibility (Variables|Default|GetSM_BananaBusAnimated) false false)
          (Components|Activation|Deactivate (Variables|Default|GetNS_WildEstate_BananaBus_MotionBlur)))
        (:Whacked
          (CallFunction|RakeWhack))))
    (:Remote
      (Variables|/Script/Engine.PrimitiveComponent|SetbGenerateOverlapEvents false (Variables|Default|GetRakeOverlap))
      (Rendering|SetVisibility (Variables|Debug|GetStaticMeshComponent) false false)
      (Rendering|SetVisibility (Variables|Default|GetSM_BananaBusAnimated) true false)
      (Components|Activation|Activate (Variables|Default|GetNS_WildEstate_BananaBus_MotionBlur) true)
      (Timeline|RakeSnapAnimation PlayFromStart))))

(event Custom|BananaBusSetModeToStun
  (Variables|Default|SetbUseStunVariant true)
  (Variables|Default|SetbUseKnockbackVariant false)
  (Development|PrintString "BANANABUS: STUN mode Activated!" true true "(R=0.056793,G=1.000000,B=0.000000,A=1.000000)" 6.0))

(event Custom|BananaBusSetModeToKnockback
  (Variables|Default|SetbUseStunVariant false)
  (Variables|Default|SetbUseKnockbackVariant true)
  (Development|PrintString "BANANABUS: KNOCKBACK mode Activated!" true true "(R=0.056793,G=1.000000,B=0.000000,A=1.000000)" 6.0))

(event Custom|BananaBusSetModeToNone
  (Variables|Default|SetbUseStunVariant false)
  (Variables|Default|SetbUseKnockbackVariant false)
  (Development|PrintString "BANANABUS: NO EFFECT mode Activated!" true true "(R=0.056793,G=1.000000,B=0.000000,A=1.000000)" 6.0))

(event Custom|RakeWhack
  ;[0 0.08588 0.15 0.5] RakeSnapHit
  (Networking|SwitchHasAuthority
    (:Authority
      _)
    (:Remote
      ;;[1 1 1] Player whacked audio
      (CallFunction|PlayWackedAudio)))
  (GameplayCue|ExecuteGameplayCueOnActor(Burst) (Variables|BananaBus|Whackback|GetWhackedPawn) (Variables|Default|GetGCNTag_WhackImpact) (Ability|GameplayCue|MakeGameplayCueParameters :NormalizedMagnitude 0.0 :RawMagnitude 0.0 :Location (Transformation|GetWorldLocation (Variables|Default|GetNS_WildEstate_BananaBus_Impact)) :Normal "0,0,0" :GameplayEffectLevel 1 :AbilityLevel 1 :bReplicateLocationWhenUsingMinimalRepProxy true))
  (if (Variables|Default|GetbUseStunVariant)
    (CallFunction|RakeWhackStun)
    (elif (Variables|Default|GetbUseKnockbackVariant)
      (CallFunction|RakeWhackKnockback))))

(event Custom|RakeWhackStun
  ;[0.4219 0.1827 0.03797] RakeWhackStun
  ;;[1 1 1] Play hit react animation
  (bind _returnvalue (Animation|PlayAnimMontage (Variables|BananaBus|Whackback|GetWhackedPawn) (CallFunction|GetDirectionalHitReactAnim) 1.0 "None"))
  ;;[1 1 1] Timer for length of stunning player. Should last until the montage is completed
  (bind _returnvalue_2 (Utilities|Time|SetTimerbyEvent (EventDispatchers|CreateEvent :!event UnstunPlayer) _returnvalue))
  (Variables|BananaBus|Whackback|SetTH_UnstunDelayTimer _returnvalue_2)
  (Weapon|SetHolsterWeapon (Variables|BananaBus|Whackback|GetWhackedPawn) 0 true false "None")
  (Networking|SwitchHasAuthority
    (:Authority
      _)
    (:Remote
      (Components|Activation|Deactivate (Variables|Default|GetNS_WildEstate_BananaBus_Impact))))
  (bind _whackedpawn (Variables|BananaBus|Whackback|GetWhackedPawn)
    (:then
      (Components|Movement|StopMovementImmediately (Variables|/Script/Engine.Character|GetCharacterMovement _whackedpawn))
      (Input|DisableInput _whackedpawn (Fortnite|GetFortPlayerControllerfromActor _whackedpawn)))
    (:else)))

(event Custom|RakeWhackKnockback
  ;[1 1 1] RakeWhackKnockback
  (CallFunction|WhackLaunchPlayer))

(event Custom|WhackLaunchPlayer
  (bind _whackedpawn (Variables|BananaBus|Whackback|GetWhackedPawn))
  (bind _returnvalue (Pawn|GetController _whackedpawn))
  ;[0 0.4531 0.2084] Launch Player that stepped on rake
  ;;[1 1 1] Players should not be impulsed if on vehicle
  (bind _whackedpawn_2 (Variables|BananaBus|Whackback|GetWhackedPawn)
    (:then
      (if (not (Vehicle|IsinVehicle _whackedpawn_2))
        (Utilities|FlowControl|Sequence
          (:then_0
            (Networking|SwitchHasAuthority
              (:Authority
                (Ability|Advanced|GrantandActivateAbilityonSelf (Ability|GetAbilitySystemComponent (Variables|BananaBus|Whackback|GetWhackedPawn)) (Variables|BananaBus|Whackback|GetGA_WhackedKnockback) self))))
          (:then_1
            ;;[1 1 1] Play launch audio and trail vfx
            (Effects|SpawnFXSystemAttached (Variables|BananaBus|VFX|GetRakeWhackFX) (Variables|/Script/Engine.Character|GetMesh (Variables|BananaBus|Whackback|GetWhackedPawn)))
            (Components|Activation|Deactivate (Variables|Default|GetNS_WildEstate_BananaBus_Impact)))
          (:then_2
            ;;[0.03135 0.01734 0.15] Send Quest Event On Knockback
            (Utilities|IsValid _returnvalue
              (:Is%20Valid
                (Ability|SendGameplayEventtoActor _returnvalue (Variables|BananaBus|Quests|GetQuest_KnockbackWhacked) (Utilities|Struct|MakeGameplayEventData (Variables|BananaBus|Quests|GetQuest_KnockbackWhacked) self _returnvalue 0 0 "()" "(GameplayTags=)" (GameplayTags|GetOwnedGameplayTags (Ability|GetFortAbilitySystemComponent _returnvalue) :!class /Script/GameplayTags.BlueprintGameplayTagLibrary)))))))))
    (:else)))

(event Custom|LaunchPlayersOnStick
  (bind _whackedpawn (Variables|BananaBus|Whackback|GetWhackedPawn))
  (bind _returnvalue (Pawn|GetController _whackedpawn))
  ;[0.15 0.05318 0.1035 0.5] Launch Players who are on other end of rake when it is stepped on
  (bind _pawnstolaunch (CallFunction|CheckOverlapforPawnstoLaunch false))
  (for _array_element _pawnstolaunch
    (bind _asfort_player_pawn (Utilities|Casting|CastToFortPlayerPawn _array_element))
    (if (not (Vehicle|IsinVehicle _asfort_player_pawn))
      (Utilities|FlowControl|Sequence
        (:then_0
          ;;[1 1 1] Launch player vertically
          (Networking|SwitchHasAuthority
            (:Authority
              (Ability|Advanced|GrantandActivateAbilityonSelf (Ability|GetAbilitySystemComponent _asfort_player_pawn) (Variables|Default|GetGA_LaunchKnockback) self))
            (:Remote
              _))
          (Collision|IgnoreActorwhenMoving (Variables|/Script/Engine.Character|GetCapsuleComponent _asfort_player_pawn) (Variables|BananaBus|Whackback|GetWhackedPawn) true) ; TEMP - should move to a GA that can track when the player lands to re-enable all collision
          ;[0.03135 0.01734 0.15] Send Quest Event On Knockback
          (Utilities|IsValid _returnvalue
            (:Is%20Valid
              (Ability|SendGameplayEventtoActor _returnvalue (Variables|Default|GetQuest_Launched) (Utilities|Struct|MakeGameplayEventData (Variables|Default|GetQuest_Launched) self _returnvalue 0 0 "()" "(GameplayTags=)" (GameplayTags|GetOwnedGameplayTags (Ability|GetFortAbilitySystemComponent _returnvalue) :!class /Script/GameplayTags.BlueprintGameplayTagLibrary))))))
        (:then_1
          ;;[1 1 1] Play launch trail vfx
          (Effects|SpawnFXSystemAttached (Variables|BananaBus|VFX|GetRakeWhackFX) (Variables|/Script/Engine.Character|GetMesh _asfort_player_pawn)))))))

(event Custom|PlayWackedAudio
  ;[0.15 0.01797 0.1351 0.5] Sounds
  (Networking|CanExecuteCosmeticEvents
    (:True
      (SoundLibrary|PlayEvent (Variables|Default|GetSoundLibrary) "(TagName=\"SoundLibrary.Item.Interact.Complete\")"))))

(event Custom|PlayDropAudio
  ;[0.15 0.01797 0.1351 0.5] Sounds
  (Networking|CanExecuteCosmeticEvents
    (:True
      (SoundLibrary|PlayEvent (Variables|Default|GetSoundLibrary) "(TagName=\"SoundLibrary.Item.Drop.Impact\")"))))

(fn CheckOverlapForPawnsToLaunch
  (params (DrawOverlapDebug :type Boolean))
  (bind _returnvalue (Transformation|GetActorLocation))
  (bind _returnvalue_2 (Transformation|GetActorRightVector))
  (bind _returnvalue_3 (Transformation|GetActorRotation))
  (bind _launchpositionoffsetmagnitude (Variables|Default|GetLaunchPositionOffsetMagnitude))
  (bind _returnvalue_4 (* _returnvalue_2 _launchpositionoffsetmagnitude))
  (bind _returnvalue_5 (+ _returnvalue _returnvalue_4))
  (bind _launchboxextent (Variables|Default|GetLaunchBoxExtent))
  (if DrawOverlapDebug
    (Rendering|Debug|DrawDebugBox _returnvalue_5 _launchboxextent "(R=1.000000,G=0.076501,B=0.487516,A=1.000000)" _returnvalue_3 3.0 2.0))
  (bind _outactors (Collision|BoxOverlapActorsWithOrientation _returnvalue_5 _launchboxextent _returnvalue_3 (Utilities|Array|MakeArray "ObjectTypeQuery3") 0 (Utilities|Array|MakeArray (Variables|BananaBus|Whackback|GetWhackedPawn))))
  (return _outactors))

(fn GetDirectionalHitReactAnim
  ;[1 1 1] Check against the backward vector
  (bind _output_get (LocalVariables|SethForwardBackDOT (Math|Vector|DotProduct (Transformation|GetActorRightVector) (Transformation|GetActorForwardVector (Variables|BananaBus|Whackback|GetWhackedPawn))))) ; store the dot to compare against \nforward if not backward
  (if (> _output_get (Variables|BananaBus|Animation|GetHitReactionForwardBackDotThreshold))
    (return (Variables|BananaBus|Animation|GetDirectionalHitAnimBackward))
    (elif (< (LocalVariables|GethForwardBackDOT) (* (Variables|BananaBus|Animation|GetHitReactionForwardBackDotThreshold) -1.0))
      (return (Variables|BananaBus|Animation|GetDirectionalHitAnimForward))
      (elif (> (Math|Vector|DotProduct (Transformation|GetActorForwardVector) (Transformation|GetActorForwardVector (Variables|BananaBus|Whackback|GetWhackedPawn))) 0)
        (return (Variables|BananaBus|Animation|GetDirectionalHitAnimLeft))
        (else
          (return (Variables|BananaBus|Animation|GetDirectionalHitAnimRight)))))))

(fn UnstunPlayer
  (bind _whackedpawn (Variables|BananaBus|Whackback|GetWhackedPawn)
    (:then
      (Weapon|SetHolsterWeapon _whackedpawn 0 false false "None")
      (Pawn|Components|CharacterMovement|SetMovementMode (Variables|/Script/Engine.Character|GetCharacterMovement _whackedpawn) "MOVE_Walking" 0)
      (Input|EnableInput _whackedpawn (Fortnite|GetFortPlayerControllerfromActor _whackedpawn))
      (Components|Activation|Deactivate (Variables|Default|GetNS_WildEstate_BananaBus_Impact)))
    (:else)))

(fn ConstructionScript
  (|CallParentFunction))
