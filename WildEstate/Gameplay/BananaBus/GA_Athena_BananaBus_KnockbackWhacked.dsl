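; Banana Bus knockback ability that aims the launch of a whacked target along the bus's own launch direction.
; It overrides GetInitialKnockMag, casts the knockback launcher to BP_BananaBus and feeds the forward vector of its LaunchDirection component to the parent calculation.
; If the cast fails, it prints an on-screen warning and passes the incoming Direction to the parent instead.
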(fn GetInitialKnockMag
  (params (Direction :type Vector :default "0, 0, 0") (OverrideXYMag :type Boolean) (XYMAgOverride :type Float%20%28double-precision%29))
  (bind _asbp_banana_bus (Utilities|Casting|CastToBP_BananaBus (Variables|KnockbackParent|GetLauncher))
    (:then
      ;[1 0.5449 0.9574] Return: Launch Magnitude
      ; uses initial direction, XY velocities, and Z velocity
      (return (|CallParentFunction (Math|Rotator|GetRotationXVector (Transformation|GetWorldRotation (Variables|/WildEstate/Gameplay/BananaBus/BP_BananaBus.BP_BananaBus_C|GetLaunchDirection _asbp_banana_bus))))))
    (:CastFailed
      ;[0 0 0] Fallback: Use default logic and do the best launch possible
      (Development|PrintString "BANANABUS: Unable to get proper launch direction; Using fallback..." true true "(R=1.000000,G=0.136484,B=0.027852,A=1.000000)" 8.0)
      ;;[1 0.5449 0.9574] Return: Default value
      (return (|CallParentFunction Direction)))))
