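# MetaSound Source for the BananaBus drop sound, a mono one-shot drawing on six drop wave variations.
# A MetaSoundWizard randomized-array one-shot node does the work, feeding its audio and finished trigger straight to the outputs.
# Min pitch and volume inputs drive that node's random Base, while Max inputs drive its Spread.
# Playback uses Fortnite pickaxe-impact attenuation and the gadget items sound class.
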MetaSoundSource /WildEstate/Gameplay/BananaBus/Sounds/MSS_BananaBus_Drop
inputs [
 Max Pitch {type=Float}
 Max Volume {type=Float}
 Min Pitch {type=Float}
 Min Volume {type=Float}
 Sounds {type=WaveAsset:Array}
 UE.Source.OnPlay {type=Trigger}
]
interfaces [UE.OutputFormat.Mono:1.0, UE.Source.OneShot:1.0, UE.Source:1.1]
outputs [
 "UE.OutputFormat.Mono.Audio:0" {type=Audio}
 UE.Source.OneShot.OnFinished {type=Trigger}
]
pages [Default {
  input_defaults {
   Max Pitch=1.5
   Max Volume=0.4
   Min Pitch=0.95
   Min Volume=0.232
   UE.Source.OnPlay=null
   Sounds [
    &/BananaBus_Drop_01
    &/BananaBus_Drop_02
    &/BananaBus_Drop_03
    &/BananaBus_Drop_04
    &/BananaBus_Drop_05
    &/BananaBus_Drop_06
   ]
  }
  nodes [MSS_Oneshot_Array_Randomized_Mono_0 = /MetaSoundWizard/MetaSounds/One-shots/ArrayRandomized/MSS_Oneshot_Array_Randomized_Mono(
    Random.Pitch.Spread=^Max Pitch
    Random.Volume.Spread=^Max Volume
    Random.Pitch.Base=^Min Pitch
    Random.Volume.Base=^Min Volume
    Sounds=^Sounds
    UE.Source.OnPlay=^UE.Source.OnPlay
   )]
  links [
   {dst="@UE.OutputFormat.Mono.Audio:0", src=MSS_Oneshot_Array_Randomized_Mono_0.UE.OutputFormat.Mono.Audio:0}
   {dst="@UE.Source.OneShot.OnFinished", src=MSS_Oneshot_Array_Randomized_Mono_0.UE.Source.OneShot.OnFinished}
  ]
 }]
properties {
 assetImportData=&self:AssetImportData
 attenuationSettings=/Game/Sounds/Attenuation/Fort_Pickaxe_Impact_Att
 maxDistance=4300
 sampleRate=48000
 soundClassObject=&Fort_Gadgets_Items
 volume=1.3
}
